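# Control Rig for the StrayMantaExtra face accessory that adjusts the dyn_hair_base hair bone from head movement.
# A spherical pose reader on the head, with debug drawing left on, weights a corrective transform.
# A second corrective multiplies that weight by a remap of head Z rotation relative to spine_05.
# A final full-weight step rotates the bone by scaled head X rotation, using a smaller range when isPlayingEmote is set.
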ControlRigBlueprint /BRCosmetics/Characters/Player/Male/Medium/Bodies/M_MED_StrayMantaExtra/Meshes/Parts/M_MED_StrayMantaExtra_FaceAcc_CtrlRig
graphs [
 RigVMFunctionLibrary {links [], nodes []}
 RigVMModel {nodes [
   RigUnit_BeginExecution = &RigUnit_BeginExecution()
   n1 = &RigUnit_SphericalPoseReader(
    ActiveRegionScaleFactors=(PositiveWidth=1.6,NegativeWidth=0.5,PositiveHeight=3,NegativeHeight=3)
    ActiveRegionSize=0.05
    Debug=(bDrawDebug=true,bDraw2D=False,bDrawLocalAxes=False,DebugScale=25,DebugSegments=20,DebugThickness=0.25)
    DriverAxis=(X=0,Y=-1,Z=0)
    DriverItem=(Type=Bone,Name="head")
    FalloffRegionScaleFactors=(PositiveWidth=1.15,NegativeWidth=0.5,PositiveHeight=0.7,NegativeHeight=0.7)
    FalloffSize=0.23
    FlipHeightScaling=False
    FlipWidthScaling=False
    OptionalParentItem=(Type=Bone,Name="None")
    RotationOffset=(X=-90,Y=0,Z=-55)
    ExecutePin=@RigUnit_BeginExecution.ExecutePin
   )
   > &RigUnit_ModifyTransforms(
    ItemToModify=((Item=(Type=Bone,Name="dyn_hair_base"),Transform=(Rotation=(X=0,Y=-0.5,Z=0,W=0.866),Translation=(X=-4,Y=0,Z=2),Scale3D=(X=0.9,Y=0.9,Z=0.9))))
    WeightMinimum=0
    Weight=@n1.OutputParam
   )
   RigVMFunction_MathQuaternionToEuler = &RigVMFunction_MathQuaternionToEuler(
    RotationOrder=ZYX
    Value=&RigUnit_GetRelativeTransformForItem(Child=(Type=Bone,Name="head"), Parent=(Type=Bone,Name="spine_05"), bChildInitial=False, bParentInitial=False).RelativeTransform.Rotation
   )
   Remap = &RigVMFunction_MathDoubleRemap(SourceMaximum=-90, SourceMinimum=-75, TargetMaximum=1, TargetMinimum=0, Value=@RigVMFunction_MathQuaternionToEuler.Result.Z)
   Multiply = &RigVMFunction_MathDoubleMul(B=@Remap.Result, A=@n1.OutputParam)
   > &RigUnit_ModifyTransforms(
    ItemToModify=((Item=(Type=Bone,Name="dyn_hair_base"),Transform=(Rotation=(X=0,Y=-0.6088,Z=0,W=0.7934),Translation=(X=1,Y=0,Z=0),Scale3D=(X=1,Y=1,Z=1))))
    WeightMinimum=0
    Weight=@Multiply.Result
   )
   RigVMFunction_MathBoolToInteger = &RigVMFunction_MathBoolToInteger(Value=RigVMVariableNode() {variable=isPlayingEmote}.Value)
   RigVMFunction_MathQuaternionFromEuler = &RigVMFunction_MathQuaternionFromEuler(
    Euler=(X=0,Y=0,Z=0)
    Euler.X=RigVMDispatchNode(
     Values=(0,0)
     Values.0=&RigVMFunction_MathDoubleRemap(SourceMaximum=90, SourceMinimum=-90, TargetMaximum=45, TargetMinimum=-45, Value=@RigVMFunction_MathQuaternionToEuler.Result.X).Result
     Values.1=&RigVMFunction_MathDoubleRemap(SourceMaximum=90, SourceMinimum=-90, TargetMaximum=30, TargetMinimum=-30, Value=@RigVMFunction_MathQuaternionToEuler.Result.X).Result
     Index=@RigVMFunction_MathBoolToInteger.Result
    ) {template="DISPATCH_RigVMDispatch_SelectInt32(in Index,in Values,out Result)"}.Result
   )
   > &RigUnit_ModifyTransforms(
    ItemToModify=((Item=(Type=Bone,Name="dyn_hair_base"),Transform=(Rotation=(X=0,Y=0,Z=0,W=1),Translation=(X=1,Y=0,Z=0),Scale3D=(X=0.9,Y=0.9,Z=0.9))))
    Weight=1
    WeightMinimum=0
    ItemToModify.0.Transform.Rotation=@RigVMFunction_MathQuaternionFromEuler.Result
   )
  ]}
]
hierarchy [
 root {loc=<0 0 0>, rot=<0 0 0 1>}
 pelvis {parent=root, loc=<0 -0.7071 93.1191>, rot=<0.0127 0.707 -0.0127 -0.707>}
 spine_01 {parent=pelvis, loc=<6.7494 -0.6388 0>, rot=<0 0 0.034 -0.9994>}
 spine_02 {parent=spine_01, loc=<7.2518 -0.035 0>, rot=<0 0 0.0238 -0.9997>}
 spine_03 {parent=spine_02, loc=<6.0104 0 0>, rot=<0 0 0.0149 -0.9999>}
 spine_04 {parent=spine_03, loc=<8.1024 0 0>, rot=<0 0 -0.0132 -0.9999>}
 spine_05 {parent=spine_04, loc=<10.0146 0 0>, rot=<0 0 -0.0015 -1>}
 neck_01 {parent=spine_05, loc=<11.3138 0.1529 0>, rot=<0 0 -0.2003 -0.9797>}
 neck_02 {parent=neck_01, loc=<4.5627 0 0>, rot=<0 0 0.0504 -0.9987>}
 head {parent=neck_02, loc=<5.3914 0 0>, rot=<0 0 0.115 -0.9934>}
 hair_base {parent=head, loc=<0 0 0>, rot=<0 -0.7071 0 -0.7071>}
 dyn_hair02 {parent=hair_base, loc=<-1.4929 1.1282 15.7243>, rot=<0.1358 0.1946 0.5509 -0.8001>}
 dyn_hair01 {parent=hair_base, loc=<3.8851 3.5432 14.2228>, rot=<0.187 0.1618 0.7343 -0.6321>}
 dyn_hair_base {parent=hair_base, loc=<0 -5.5183 0.7012>, rot=<-0.4481 -0.4481 0.547 -0.547>}
 dyn_hair_m {parent=dyn_hair_base, loc=<0 0 0>, rot=<0 0 0 1>}
 dyn_hair01_r {parent=dyn_hair_m, loc=<3.1291 -3.8895 0.0646>, rot=<-0.297 0.2672 0.1535 -0.9038>}
 dyn_hair012_r {parent=dyn_hair01_r, loc=<2.0308 0.1532 -0.1044>, rot=<-0.0146 0.2937 0.0261 -0.9554>}
 dyn_hair011_r {parent=dyn_hair01_r, loc=<-0.1541 -1.9308 0.3227>, rot=<0.036 0.1396 0.0957 -0.9849>}
 dyn_hair01_m {parent=dyn_hair_m, loc=<2.8916 0.0903 0.1922>, rot=<0 0.2539 0 -0.9672>}
 dyn_hair011_m {parent=dyn_hair01_m, loc=<1.6454 -1.0357 -0.1017>, rot=<-0.045 0.1131 0.1385 -0.9828>}
 dyn_hair011_l {parent=dyn_hair01_m, loc=<2.3178 1.8869 0.3185>, rot=<0.1868 0.2068 -0.2381 -0.9304>}
 dyn_hair01_l {parent=dyn_hair_m, loc=<3.1294 3.8895 0.0646>, rot=<0.297 0.2672 -0.1535 -0.9038>}
 dyn_hair012_l {parent=dyn_hair01_l, loc=<1.7928 -0.1214 -0.3065>, rot=<-0.0854 0.2605 -0.0543 -0.9602>}
 dyn_hair013_l {parent=dyn_hair01_l, loc=<0.7024 1.8628 0.5275>, rot=<0.0031 0.1088 -0.1056 -0.9884>}
 dyn_hair021_l {parent=dyn_hair_m, loc=<1.3147 1.9137 2.1718>, rot=<-0.0793 0.6484 0.0044 -0.7572>}
 dyn_hair021_r {parent=dyn_hair_m, loc=<2 -2.7574 1.9594>, rot=<-0.2055 0.2989 0.2564 -0.8959>}
 dyn_hair021_m {parent=dyn_hair_m, loc=<1.3543 -0.2619 2.1673>, rot=<0.0631 0.351 -0.0616 -0.9322>}
 dyn_hair023_l {parent=dyn_hair_m, loc=<0.5781 4.2201 1.3876>, rot=<0.1609 0.4178 -0.2086 -0.8695>}
 dyn_hair022_l {parent=dyn_hair_m, loc=<2.1083 3.4775 1.4036>, rot=<0.1078 0.481 -0.1862 -0.8499>}
 dyn_hair027_l {parent=dyn_hair_m, loc=<0.9564 6.4424 -0.8891>, rot=<0.25 0.421 -0.3471 -0.7998>}
 dyn_hair023_r {parent=dyn_hair_m, loc=<0.5822 -6.115 -1.2569>, rot=<-0.0712 0.3444 0.4064 -0.8433>}
 dyn_hair022_r {parent=dyn_hair_m, loc=<0.4712 -4.3312 1.5747>, rot=<-0.1656 0.368 0.1875 -0.8956>}
 dyn_hair025_l {parent=dyn_hair_m, loc=<-0.5667 4.9705 1.3143>, rot=<0.2188 0.398 -0.1526 -0.8778>}
 dyn_hair_back_01 {parent=hair_base, loc=<-3.965 -7.3021 6.2598>, rot=<-0.1674 -0.0802 0.8905 -0.4154>}
 dyn_hair_back_02 {parent=hair_base, loc=<4.2874 -7.7427 6.1121>, rot=<-0.1058 -0.164 0.3728 -0.9072>}
 dyn_hairTop_03 {parent=hair_base, loc=<-5.2997 -2.7849 12.6543>, rot=<-0.1153 -0.1428 0.6815 -0.7085>}
 dyn_hairTop_02 {parent=hair_base, loc=<-1.9046 -5.171 13.7656>, rot=<-0.2308 -0.2084 0.6031 -0.7345>}
 dyn_hairTop_01 {parent=hair_base, loc=<2.4262 -1.7268 15.1776>, rot=<-0.1233 -0.1911 0.5654 -0.7928>}
 dyn_hair024_r {parent=hair_base, loc=<-4.2704 -7.8395 2.3667>, rot=<-0.0754 -0.057 0.8415 -0.5319>}
 dyn_hair026_l {parent=hair_base, loc=<5.582 -5.8793 2.9584>, rot=<-0.1742 -0.1461 0.3339 -0.9148>}
 dyn_hair025_r {parent=hair_base, loc=<-5.4881 -5.8513 2.8367>, rot=<-0.1172 -0.0331 0.9473 -0.2964>}
 dyn_hairCurl_01 {parent=hair_base, loc=<-3.3055 7.5135 11.0201>, rot=<-0.5351 -0.517 0.663 -0.0833>}
 dyn_hairCurl_02 {parent=dyn_hairCurl_01, loc=<1.8248 0 0>, rot=<-0.0151 -0.0081 -0.2416 -0.9702>}
 dyn_hairCurl_03 {parent=dyn_hairCurl_02, loc=<1.95 0 0>, rot=<0.0438 0.0782 -0.4082 -0.9085>}
 clavicle_l {parent=spine_05, loc=<4.3575 -0.1466 -2.3223>, rot=<0.0465 -0.7289 -0.009 -0.6829>}
 upperarm_l {parent=clavicle_l, loc=<15.5111 0 0>, rot=<0.0226 -0.3234 -0.0088 -0.9459>}
 upperarm_twist_01_l {parent=upperarm_l, loc=<0.0041 0 0>, rot=<0 0 0 -1>}
 deform_upperarm_fr_base_l {parent=upperarm_twist_01_l, loc=<-0.3691 7.0506 -2.765>, rot=<0 0 0 1>}
 deform_upperarm_fr_l {parent=deform_upperarm_fr_base_l, loc=<0 0 0>, rot=<0 0 0 1>}
 deform_upperarm_up_base_l {parent=upperarm_twist_01_l, loc=<-3.2216 -1.1391 3.3311>, rot=<0 0 0 1>}
 deform_upperarm_up_l {parent=deform_upperarm_up_base_l, loc=<0 0 0>, rot=<0 0 0 1>}
 clavicle_r {parent=spine_05, loc=<4.3575 -0.1466 2.3223>, rot=<0.7289 0.0465 -0.6829 0.009>}
 upperarm_r {parent=clavicle_r, loc=<-15.5111 0 0>, rot=<0.0226 -0.3234 -0.0088 -0.9459>}
 upperarm_twist_01_r {parent=upperarm_r, loc=<-0.0041 0 0>, rot=<0 0 0 -1>}
 deform_upperarm_fr_base_r {parent=upperarm_twist_01_r, loc=<0.3689 -7.0506 2.7652>, rot=<0 0 0 1>}
 deform_upperarm_fr_r {parent=deform_upperarm_fr_base_r, loc=<0 0 0>, rot=<0 0 0 1>}
 deform_upperarm_up_base_r {parent=upperarm_twist_01_r, loc=<3.2214 1.1391 -3.3308>, rot=<0 0 0 1>}
 deform_upperarm_up_r {parent=deform_upperarm_up_base_r, loc=<0 0 0>, rot=<0 0 0 -1>}
 hood_correct {parent=spine_05, loc=<18.3883 -8.3616 0>, rot=<0.1957 -0.7084 0.6451 -0.2092>}
 deform_pec_aim_base_l {parent=spine_04, loc=<7.8732 12.6388 -2.4123>, rot=<0.0102 -0.4602 0.3979 -0.7936>}
 deform_pec_aim_l {parent=deform_pec_aim_base_l, loc=<0 0 0>, rot=<0 0 0 1>}
 deform_pec_base_l {parent=deform_pec_aim_l, loc=<10.766 0 0>, rot=<0 0 0 1>}
 deform_pec_l {parent=deform_pec_base_l, loc=<0.055 0 0>, rot=<0 0 0 1>}
 deform_pec_aim_base_r {parent=spine_04, loc=<7.8727 12.6388 2.4123>, rot=<-0.4602 -0.0102 0.7936 0.3979>}
 deform_pec_aim_r {parent=deform_pec_aim_base_r, loc=<0 0 0>, rot=<0 0 0 1>}
 deform_pec_base_r {parent=deform_pec_aim_r, loc=<-10.7662 -0.0006 -0.0007>, rot=<0 0 0 1>}
 deform_pec_r {parent=deform_pec_base_r, loc=<0 0 0>, rot=<0 0 0 1>}
]
variables [isPlayingEmote {type=bool, default=False}]
