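# Physics asset for the face accessory part of the StrayMantaExtra medium male player cosmetic, defining bodies and joints for a small dynamic hair rig.
# Capsules on spine_05, both clavicles and neck_01 cover the upper body, joined by angularly free constraints.
# A hair base sphere hangs off the neck with free rotation, feeding a middle capsule that branches into left, middle and right tips.
# Hair joints lock twist, and each tip swings on only one limited axis.
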PhysicsAsset /BRCosmetics/Characters/Player/Male/Medium/Bodies/M_MED_StrayMantaExtra/Meshes/Parts/M_MED_StrayMantaExtra_FaceAcc_Physics
bodies {
 clavicle_l [{
   shapeType=Capsule
   length=4.2552
   radius=8.1048
   shapeName=clavicle_l_capsule
   center=<8.6321 -0.2957 0.0402>
   rotation=<rot -0.0502 -86.313 98.5709>
  }]
 clavicle_r [{
   shapeType=Capsule
   length=4.2552
   radius=8.1048
   shapeName=clavicle_r_capsule
   center=<-8.6321 0.2957 -0.0402>
   rotation=<rot -0.0502 -86.313 -81.4291>
  }]
 dyn_hair01_l [{
   shapeType=Capsule
   length=2.7211
   radius=1.9102
   shapeName=dyn_hair01_l_capsule
   center=<1.8005 0 0.6826>
   rotation=<rot -69.2383 0.0001 -0.0001>
  }]
 dyn_hair01_m [{
   shapeType=Capsule
   length=3.9565
   radius=2.274
   shapeName=dyn_hair01_m_capsule
   center=<3.243 -0.1966 0.7326>
   rotation=<rot -0.0001 0 -90.33>
  }]
 dyn_hair01_r [{
   shapeType=Capsule
   length=2.7211
   radius=1.9102
   shapeName=dyn_hair01_r_capsule
   center=<1.8269 0 0.5529>
   rotation=<rot -78.527 -0.0001 -179.9999>
  }]
 dyn_hair_base [{shapeType=Sphere, radius=0.505, shapeName=dyn_hair_base_sphere}]
 dyn_hair_m [{
   shapeType=Capsule
   length=4.5647
   radius=2.1598
   shapeName=dyn_hair_m_capsule
   center=<1.9025 -0.0104 0.8151>
   rotation=<rot -0.0003 0.477 -90>
  }]
 neck_01 [{shapeType=Capsule, length=5.7894, radius=5.808, shapeName=neck_01_capsule, center=<0 0.7442 0>, rotation=<rot 0 -90.0001 89.9646>}]
 spine_05 [{shapeType=Capsule, length=11.1481, radius=12.5287, shapeName=spine_05_capsule, center=<-0.8508 -1.0672 0>, rotation=<rot 0 -90 0>}]
}
constraints [
 {
  bone1Name=clavicle_l
  bone2Name=spine_05
  swing1LimitDegrees=45
  swing1Motion=Free
  swing2LimitDegrees=45
  swing2Motion=Free
  twistLimitDegrees=45
  twistMotion=Free
 }
 {
  bone1Name=clavicle_r
  bone2Name=spine_05
  swing1LimitDegrees=45
  swing1Motion=Free
  swing2LimitDegrees=45
  swing2Motion=Free
  twistLimitDegrees=45
  twistMotion=Free
 }
 {
  bone1Name=neck_01
  bone2Name=spine_05
  swing1LimitDegrees=45
  swing1Motion=Free
  swing2LimitDegrees=45
  swing2Motion=Free
  twistLimitDegrees=45
  twistMotion=Free
 }
 {
  bone1Name=dyn_hair_base
  bone2Name=neck_01
  swing1LimitDegrees=45
  swing1Motion=Free
  swing2LimitDegrees=45
  swing2Motion=Free
  twistLimitDegrees=45
  twistMotion=Free
 }
 {bone1Name=dyn_hair_m, bone2Name=dyn_hair_base, swing1LimitDegrees=10, swing2LimitDegrees=12, twistLimitDegrees=45, twistMotion=Locked}
 {
  bone1Name=dyn_hair01_l
  bone2Name=dyn_hair_m
  swing1LimitDegrees=45
  swing1Motion=Locked
  swing2LimitDegrees=15
  twistLimitDegrees=45
  twistMotion=Locked
 }
 {
  bone1Name=dyn_hair01_m
  bone2Name=dyn_hair_m
  swing1LimitDegrees=45
  swing1Motion=Locked
  swing2LimitDegrees=15
  twistLimitDegrees=45
  twistMotion=Locked
 }
 {
  bone1Name=dyn_hair01_r
  bone2Name=dyn_hair_m
  swing1LimitDegrees=45
  swing1Motion=Locked
  swing2LimitDegrees=15
  twistLimitDegrees=45
  twistMotion=Locked
 }
]
